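# Physics asset for the FrontEnd variant of the male medium Snow Ninja belt, defining how its left and right hanging pieces swing.
# Each side is a four-bone chain with swing and tight twist limits between segments, its root angularly locked to the pelvis.
# Paired side-by-side capsules on the lower segments approximate a flat, wide shape.
# Pelvis and thigh capsules, also angularly locked together, supply body volumes for the belt.
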PhysicsAsset /BRCosmetics/Characters/Player/Male/Medium/Bodies/M_MED_Snow_Ninja/Meshes/M_MED_Snow_Ninja_Belt_FrontEnd_Physics
bodies {
 dyn_belt_01_l [{
   shapeType=Capsule
   length=0.505
   radius=0.505
   shapeName=dyn_belt_01_l_capsule
   center=<4.1449 -0.1569 -0.1523>
   rotation=<rot 0 -1.4247 -86.4681>
  }]
 dyn_belt_01_r [{
   shapeType=Capsule
   length=3.4965
   radius=2.0873
   shapeName=dyn_belt_01_r_capsule
   center=<1.9546 0.0718 0.0842>
   rotation=<rot 0 -1.7606 -88.7068>
  }]
 dyn_belt_02_l [
  {
   shapeType=Capsule
   length=8.5269
   radius=1.0276
   shapeName=dyn_belt_02_l_capsule
   center=<4.5463 -2.3416 0.2106>
   rotation=<rot 90 180 -177.9652>
  }
  {
   shapeType=Capsule
   length=8.5269
   radius=1.0276
   shapeName=dyn_belt_02_l_capsule
   center=<4.4883 1.3015 0.2106>
   rotation=<rot 90 180 178.3141>
  }
 ]
 dyn_belt_02_r [
  {
   shapeType=Capsule
   length=7.5989
   radius=0.7343
   shapeName=dyn_belt_02_l_capsule
   center=<3.806 -2.3941 0.4103>
   rotation=<rot 90 180 -178.7454>
  }
  {
   shapeType=Capsule
   length=6.8379
   radius=1.4953
   shapeName=dyn_belt_02_l_capsule
   center=<3.7246 2.2671 0.4103>
   rotation=<rot 90 180 178.4872>
  }
 ]
 dyn_belt_03_l [
  {shapeType=Capsule, length=7.7849, radius=1.6724, shapeName=dyn_belt_03_l_capsule, center=<4.0821 -1.9842 0>, rotation=<rot 90 180 -178.4078>}
  {shapeType=Capsule, length=7.7849, radius=1.6724, shapeName=dyn_belt_03_l_capsule, center=<3.9847 1.944 0>, rotation=<rot 90 0 -1.7685>}
 ]
 dyn_belt_03_r [
  {
   shapeType=Capsule
   length=7.2325
   radius=1.6645
   shapeName=dyn_belt_03_r_capsule
   center=<4.0902 -2.1701 0.5659>
   rotation=<rot 90 180 -179.0159>
  }
  {
   shapeType=Capsule
   length=7.2325
   radius=1.6645
   shapeName=dyn_belt_03_r_capsule
   center=<4 2.3758 0.5659>
   rotation=<rot 90 179.7762 177.734>
  }
 ]
 dyn_belt_04_l [
  {shapeType=Capsule, length=6.2295, radius=1.7705, shapeName=dyn_belt_04_l_capsule, center=<4.1561 2.2403 0.0625>, rotation=<rot 90 0 0>}
  {shapeType=Capsule, length=6.2295, radius=1.7705, shapeName=dyn_belt_04_l_capsule, center=<4.1561 -2.5556 0.0625>, rotation=<rot 90 0 0>}
 ]
 dyn_belt_04_r [
  {
   shapeType=Capsule
   length=6.4728
   radius=1.5271
   shapeName=dyn_belt_04_l_capsule
   center=<4.2953 -2.5355 0.7226>
   rotation=<rot 90 179.7762 -179.0154>
  }
  {
   shapeType=Capsule
   length=6.4728
   radius=1.5271
   shapeName=dyn_belt_04_l_capsule
   center=<4.1398 2.9824 0.7226>
   rotation=<rot 90 179.7016 177.7511>
  }
 ]
 pelvis [{
   shapeType=Capsule
   length=16.2213
   radius=8.8068
   shapeName=pelvis_capsule
   center=<-7.5504 -1.187 0.1295>
   rotation=<rot -0.1018 -90 0.3734>
  }]
 thigh_l [{
   shapeType=Capsule
   length=30.8091
   radius=8.1909
   shapeName=thigh_l_capsule
   center=<-16.5403 -0.8021 -0.1834>
   rotation=<rot 86.3058 -177.0005 179.9997>
  }]
 thigh_r [{shapeType=Capsule, length=30.9976, radius=8.0024, shapeName=thigh_l_capsule, center=<16.5 0.8 0.2521>, rotation=<rot 89.0681 2.9974 0.0011>}]
}
constraints [
 {bone1Name=dyn_belt_02_l, bone2Name=dyn_belt_01_l, swing1LimitDegrees=15, swing2LimitDegrees=20, twistLimitDegrees=0}
 {bone1Name=dyn_belt_03_l, bone2Name=dyn_belt_02_l, swing1LimitDegrees=40, swing2LimitDegrees=40, twistLimitDegrees=10}
 {bone1Name=dyn_belt_04_l, bone2Name=dyn_belt_03_l, swing1LimitDegrees=40, swing2LimitDegrees=40, twistLimitDegrees=10}
 {bone1Name=dyn_belt_02_r, bone2Name=dyn_belt_01_r, swing1LimitDegrees=40, swing2LimitDegrees=40, twistLimitDegrees=0}
 {bone1Name=dyn_belt_03_r, bone2Name=dyn_belt_02_r, swing1LimitDegrees=40, swing2LimitDegrees=40, twistLimitDegrees=10}
 {bone1Name=dyn_belt_04_r, bone2Name=dyn_belt_03_r, swing1LimitDegrees=40, swing2LimitDegrees=40, twistLimitDegrees=10}
 {
  bone1Name=thigh_l
  bone2Name=pelvis
  swing1LimitDegrees=45
  swing1Motion=Locked
  swing2LimitDegrees=45
  swing2Motion=Locked
  twistLimitDegrees=45
  twistMotion=Locked
 }
 {
  bone1Name=thigh_r
  bone2Name=pelvis
  swing1LimitDegrees=45
  swing1Motion=Locked
  swing2LimitDegrees=45
  swing2Motion=Locked
  twistLimitDegrees=45
  twistMotion=Locked
 }
 {
  bone1Name=dyn_belt_01_l
  bone2Name=pelvis
  swing1LimitDegrees=45
  swing1Motion=Locked
  swing2LimitDegrees=45
  swing2Motion=Locked
  twistLimitDegrees=45
  twistMotion=Locked
 }
 {
  bone1Name=dyn_belt_01_r
  bone2Name=pelvis
  swing1LimitDegrees=45
  swing1Motion=Locked
  swing2LimitDegrees=45
  swing2Motion=Locked
  twistLimitDegrees=45
  twistMotion=Locked
 }
]
