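# Physics asset for the belt of the medium male Snow Ninja player body, defining simulated motion for its left and right dangling belt pieces.
# Each side is a four-bone dyn_belt chain hung from the pelvis by swing-limited joints, with twist limited to zero degrees at the root.
# Paired side-by-side capsules on each belt bone approximate a flat strap.
# Pelvis and thigh capsules supply body volumes for the straps to collide against.
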PhysicsAsset /BRCosmetics/Characters/Player/Male/Medium/Bodies/M_MED_Snow_Ninja/Meshes/M_MED_Snow_Ninja_Belt_Physics
bodies {
 dyn_belt_01_l [
  {shapeType=Capsule, length=4, radius=1, shapeName=dyn_belt_01_l_capsule, center=<4.5112 -2.3474 0.3926>, rotation=<rot 90 179.5524 -177.6351>}
  {shapeType=Capsule, length=2.5, radius=1, shapeName=dyn_belt_01_l_capsule, center=<5.1923 0.7469 0>, rotation=<rot 90 0 0>}
 ]
 dyn_belt_01_r [
  {shapeType=Capsule, length=4.5, radius=1, shapeName=dyn_belt_01_r_capsule, center=<3.6533 2.4063 0.2948>, rotation=<rot 90 177.6321 174.8196>}
  {shapeType=Capsule, length=4, radius=1, shapeName=dyn_belt_01_r_capsule, center=<3.7806 -1.2034 0.374>, rotation=<rot 90 0 0>}
 ]
 dyn_belt_02_l [
  {
   shapeType=Capsule
   length=8.5269
   radius=1.0276
   shapeName=dyn_belt_02_l_capsule
   center=<4.5463 -2.3416 0.2106>
   rotation=<rot 90 180 -177.9652>
  }
  {
   shapeType=Capsule
   length=8.5269
   radius=1.0276
   shapeName=dyn_belt_02_l_capsule
   center=<4.4883 1.3015 0.2106>
   rotation=<rot 90 180 178.3141>
  }
 ]
 dyn_belt_02_r [
  {
   shapeType=Capsule
   length=7.0989
   radius=1.2343
   shapeName=dyn_belt_02_l_capsule
   center=<3.806 -2.3941 0.5126>
   rotation=<rot 90 180 -178.7454>
  }
  {
   shapeType=Capsule
   length=6.8379
   radius=1.4953
   shapeName=dyn_belt_02_l_capsule
   center=<3.7246 2.2671 0.4103>
   rotation=<rot 90 180 178.4872>
  }
 ]
 dyn_belt_03_l [
  {shapeType=Capsule, length=7.7849, radius=1.6724, shapeName=dyn_belt_03_l_capsule, center=<4.0821 -1.9842 0>, rotation=<rot 90 180 -178.4078>}
  {shapeType=Capsule, length=7.7849, radius=1.6724, shapeName=dyn_belt_03_l_capsule, center=<3.9847 1.944 0>, rotation=<rot 90 0 -1.7685>}
 ]
 dyn_belt_03_r [
  {
   shapeType=Capsule
   length=7.2325
   radius=1.6645
   shapeName=dyn_belt_03_r_capsule
   center=<4.0872 -2.3421 0.5659>
   rotation=<rot 90 180 -179.0159>
  }
  {
   shapeType=Capsule
   length=7.2325
   radius=1.6645
   shapeName=dyn_belt_03_r_capsule
   center=<4 2.3758 0.2507>
   rotation=<rot 90 -179.7016 178.2563>
  }
 ]
 dyn_belt_04_l [
  {shapeType=Capsule, length=6.2295, radius=1.7705, shapeName=dyn_belt_04_l_capsule, center=<4.1561 2.2403 0.0625>, rotation=<rot 90 0 0>}
  {shapeType=Capsule, length=6.2295, radius=1.7705, shapeName=dyn_belt_04_l_capsule, center=<4.1561 -2.5556 0.0625>, rotation=<rot 90 0 0>}
 ]
 dyn_belt_04_r [
  {
   shapeType=Capsule
   length=6.4728
   radius=1.5271
   shapeName=dyn_belt_04_l_capsule
   center=<4.9082 -2.7257 0.546>
   rotation=<rot 90 179.7762 -179.0154>
  }
  {
   shapeType=Capsule
   length=6.4728
   radius=1.5271
   shapeName=dyn_belt_04_l_capsule
   center=<4.1398 2.9824 -0.096>
   rotation=<rot 90 179.7016 177.7511>
  }
 ]
 pelvis [
  {shapeType=Capsule, length=14.6456, radius=9.0388, center=<-3.2627 -0.4834 0>}
  {shapeType=Capsule, length=5, radius=5, shapeName=pelvis_capsule, center=<-8.029 4.3687 0>}
 ]
 thigh_l [{shapeType=Capsule, length=32.3438, radius=10, center=<-19.4348 -1.4504 0.3808>, rotation=<rot 90 0 0>}]
 thigh_r [{shapeType=Capsule, length=32.3438, radius=10, center=<19.4348 1.2196 -0.3808>, rotation=<rot 90 180 0>}]
}
constraints [
 {bone1Name=dyn_belt_03_l, bone2Name=dyn_belt_02_l, swing1LimitDegrees=40, swing2LimitDegrees=40, twistLimitDegrees=10}
 {bone1Name=dyn_belt_04_l, bone2Name=dyn_belt_03_l, swing1LimitDegrees=40, swing2LimitDegrees=40, twistLimitDegrees=10}
 {bone1Name=dyn_belt_03_r, bone2Name=dyn_belt_02_r, swing1LimitDegrees=40, swing2LimitDegrees=40, twistLimitDegrees=10}
 {bone1Name=dyn_belt_04_r, bone2Name=dyn_belt_03_r, swing1LimitDegrees=40, swing2LimitDegrees=40, twistLimitDegrees=10}
 {bone1Name=dyn_belt_01_l, bone2Name=pelvis, swing1LimitDegrees=20, swing2LimitDegrees=40, twistLimitDegrees=0}
 {bone1Name=dyn_belt_02_l, bone2Name=dyn_belt_01_l, swing1LimitDegrees=15, swing2LimitDegrees=20, twistLimitDegrees=20}
 {bone1Name=dyn_belt_01_r, bone2Name=pelvis, swing1LimitDegrees=30, swing2LimitDegrees=20, twistLimitDegrees=0}
 {bone1Name=dyn_belt_02_r, bone2Name=dyn_belt_01_r, swing1LimitDegrees=40, swing2LimitDegrees=40, twistLimitDegrees=10}
]
