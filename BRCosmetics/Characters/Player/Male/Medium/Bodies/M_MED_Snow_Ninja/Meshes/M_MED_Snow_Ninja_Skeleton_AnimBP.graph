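# Post-process style AnimBlueprint for the Snow Ninja body, copying the parent mesh pose and layering corrective and secondary motion.
# Two Control Rigs feed front-end-switched belt offsets and RigidBody sims, then AnimDynamics for knife, straps, armbands and collar.
# Final bone tweaks fix belt, shoulders and collar, with extra parachuting offsets.
# EventGraph derives rig alphas and blend flags from weapon animation type, crouch, parachute and front-end/emote state.
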AnimBlueprint /BRCosmetics/Characters/Player/Male/Medium/Bodies/M_MED_Snow_Ninja/Meshes/M_MED_Snow_Ninja_Skeleton_AnimBP
target_skeleton=&/M_MED_Snow_Ninja_Skeleton
graphs [
 AnimGraph {nodes [
   Output Pose(Result=Use cached pose 'CRCache_PostFixesHoodie'()) {layerGroup=None}
   n1 = Control Rig!(
    Source=Copy Pose From Mesh() {bUseAttachedParent=true}
    Alpha
    dualWeaponFix:double=Get dualGunFix() -> (dualGunFix:double)
    largeWeapon:double=Get LargeGunOut() -> (LargeGunOut:double)
    bigWeapon:double=Get BigGunOut() -> (BigGunOut:double)
    collarFixAlpha:double=Get CollarFixAlpha() -> (CollarFixAlpha:double)
    rocketWeapon:double=Get RocketGunOut() -> (RocketGunOut:double)
    crouching:double=Get crouchAlpha() -> (crouchAlpha:double)
    crouchFixAlpha:double=Get crouchFixAlpha() -> (crouchFixAlpha:double)
    shotGun:double=Get ShotgunAlpha() -> (ShotgunAlpha:double)
    AR_Bull:double=Get AR_BullAlpha() -> (AR_BullAlpha:double)
    dualWeapon:double=Get dualGunOut() -> (dualGunOut:double)
   ) -> (Pose) {
    alphaBoolBlend {blendOption=Linear}
    controlRigAssetReference {blueprintRigClass=&/M_MED_Snow_Ninja_CtrlRig.M_MED_Snow_Ninja_CtrlRig_C}
   }
   n2 = Control Rig!(Source=@n1, Alpha) -> (Pose) {
    alphaBoolBlend {blendOption=Linear}
    controlRigAssetReference {blueprintRigClass=&/M_MED_Snow_Ninja_DeformRig_Master_ControlRig.M_MED_Snow_Ninja_DeformRig_Master_ControlRig_C}
   }
   Save cached pose 'CRCache_Entry'(Pose=@n2) {cacheName=CRCache_Entry}
   n3 = Anim Dynamics - Bone: dyn_knife(ComponentPose=Local To Component(LocalPose=Use cached pose 'CRCache_BeltDyn'()), Alpha=0.75, ExternalForce="0,0,0") {
    alpha=0.9
    angularDampingOverride=0.998
    angularSpringConstant=1500
    bAngularSpring=true
    bGravityOverrideInSimSpace=true
    bOverrideAngularDamping=true
    gravityScale=0.32
    numSolverIterationsPostUpdate=4
    numSolverIterationsPreUpdate=16
    simulationSpace=RootRelative
    alphaBoolBlend {blendOption=Linear}
    boundBone {boneName=dyn_knife}
    physicsBodyDefinitions [{
      sphereCollisionRadius=0
      boundBone {boneName=dyn_knife}
      boxExtents=<15 5 5>
      constraintSetup {angularLimitsMax=<4 2 5>, angularLimitsMin=<-4 -2 -5>, angularTarget=<0 0 1>}
      localJointOffset=<-0.25 0 0>
     }]
    relativeSpaceBone {boneName=root}
   }
   n4 = Anim Dynamics - Bone: dyn_armband_low(ComponentPose=@n3, Alpha=0.5, ExternalForce="0,0,0") {
    alpha=0.9
    angularDampingOverride=0.998
    angularSpringConstant=1200
    bAngularSpring=true
    bGravityOverrideInSimSpace=true
    bOverrideAngularDamping=true
    gravityScale=0.32
    numSolverIterationsPostUpdate=4
    numSolverIterationsPreUpdate=16
    simulationSpace=RootRelative
    alphaBoolBlend {blendOption=Linear}
    boundBone {boneName=dyn_armband_low}
    componentLinearAccScale=<1 0.2 0.2>
    componentLinearVelScale=<0.2 0.2 0.2>
    physicsBodyDefinitions [{
      sphereCollisionRadius=0
      boundBone {boneName=dyn_armband_low}
      boxExtents=<15 5 5>
      constraintSetup {angularTargetAxis=AxisZ, twistAxis=AxisZ, angularLimitsMax=<15 15 15>, angularLimitsMin=<-15 -15 -15>, angularTarget=<0 0 1>}
      localJointOffset=<0 0 0>
     }]
    relativeSpaceBone {boneName=root}
   }
   n5 = Anim Dynamics - Bone: dyn_armband_hi(ComponentPose=@n4, Alpha=0.5, ExternalForce="0,0,0") {
    alpha=0.9
    angularDampingOverride=0.998
    angularSpringConstant=1200
    bAngularSpring=true
    bGravityOverrideInSimSpace=true
    bOverrideAngularDamping=true
    gravityScale=0.32
    numSolverIterationsPostUpdate=4
    numSolverIterationsPreUpdate=16
    simulationSpace=RootRelative
    alphaBoolBlend {blendOption=Linear}
    boundBone {boneName=dyn_armband_hi}
    componentLinearAccScale=<1 0.2 0.2>
    componentLinearVelScale=<0.2 0.2 0.2>
    physicsBodyDefinitions [{
      sphereCollisionRadius=0
      boundBone {boneName=dyn_armband_hi}
      boxExtents=<15 5 5>
      constraintSetup {angularTargetAxis=AxisZ, twistAxis=AxisZ, angularLimitsMax=<15 15 15>, angularLimitsMin=<-15 -15 -15>, angularTarget=<0 0 1>}
      localJointOffset=<0 0 0>
     }]
    relativeSpaceBone {boneName=root}
   }
   n6 = Anim Dynamics - Bone: dyn_collar_b(ComponentPose=@n5, Alpha=1, ExternalForce="0,0,0") {
    angularDampingOverride=0.98
    bGravityOverrideInSimSpace=true
    bOverrideAngularDamping=true
    bOverrideLinearDamping=true
    gravityScale=0.1
    lODThreshold=1
    linearDampingOverride=0.98
    numSolverIterationsPostUpdate=4
    numSolverIterationsPreUpdate=16
    alphaBoolBlend {blendOption=Linear}
    boundBone {boneName=dyn_collar_b}
    componentLinearAccScale=<1 1 1>
    componentLinearVelScale=<1 1 1>
    physicsBodyDefinitions [{
      sphereCollisionRadius=0
      boundBone {boneName=dyn_collar_b}
      boxExtents=<2.5 20 20>
      constraintSetup {angularLimitsMax=<0 5 30>, angularLimitsMin=<0 -5 -30>, angularTarget=<0 0 0>}
      localJointOffset=<-0.5 0 0>
     }]
    relativeSpaceBone {boneName=root}
   }
   n7 = Anim Dynamics - Bone: dyn_knifeStrap_01_m(ComponentPose=@n6, Alpha=1, ExternalForce="0,0,0") {
    angularDampingOverride=1
    angularSpringConstant=20000
    bAngularSpring=true
    bOverrideAngularDamping=true
    bUseGravityOverride=true
    gravityScale=0.35
    simulationSpace=RootRelative
    alphaBoolBlend {blendOption=Linear}
    boundBone {boneName=dyn_knifeStrap_01_m}
    componentLinearAccScale=<1 1 1>
    componentLinearVelScale=<1 1 1>
    gravityOverride {z=-980}
    physicsBodyDefinitions [{
      boundBone {boneName=dyn_knifeStrap_01_m}
      boxExtents=<15 25 25>
      constraintSetup {angularLimitsMax=<0 25 25>, angularLimitsMin=<0 -25 -25>, angularTarget=<1 0 0>}
      localJointOffset=<4 0 0>
     }]
    relativeSpaceBone {boneName=dyn_knife}
   }
   n8 = Anim Dynamics - Bone: dyn_knifeStrap_02_m(ComponentPose=@n7, Alpha=1, ExternalForce="0,0,0") {
    angularDampingOverride=1
    angularSpringConstant=20000
    bAngularSpring=true
    bOverrideAngularDamping=true
    bUseGravityOverride=true
    gravityScale=0.35
    simulationSpace=RootRelative
    alphaBoolBlend {blendOption=Linear}
    boundBone {boneName=dyn_knifeStrap_02_m}
    componentLinearAccScale=<1 1 1>
    componentLinearVelScale=<1 1 1>
    gravityOverride {z=-980}
    physicsBodyDefinitions [{
      boundBone {boneName=dyn_knifeStrap_02_m}
      boxExtents=<15 25 25>
      constraintSetup {angularLimitsMax=<0 25 25>, angularLimitsMin=<0 -25 -25>, angularTarget=<1 0 0>}
      localJointOffset=<4 0 0>
     }]
    relativeSpaceBone {boneName=dyn_knife}
   }
   Save cached pose 'CRCache_Dyn'(Pose=Component To Local(ComponentPose=@n8)) {cacheName=CRCache_Dyn}
   n9 = "Transform (Modify) Bone - Bone: dyn_belt_r_01"!(
    ComponentPose=Local To Component(LocalPose=Use cached pose 'CRCache_Dyn'())
    bAlphaBoolEnabled:bool=True
    Translation="0,0,0"
    Rotation="0,0,-5"
    Scale="1,1,1"
   ) -> (Pose) {
    alphaInputType=Bool
    rotationMode=BMM_Additive
    alphaBoolBlend {blendOption=Linear}
    boneToModify {boneName=dyn_belt_r_01}
    rotation {roll=-5}
   }
   n10 = "Constraint - shoulder_pad_r (RR)"!(ComponentPose=@n9, bAlphaBoolEnabled:bool=True, ConstraintWeights_0=1, ConstraintWeights_1:float=0.5) -> (Pose) {
    alphaInputType=Bool
    alphaBoolBlend {blendInTime=0.1, blendOption=Linear, blendOutTime=0.1}
    boneToModify {boneName=shoulder_pad_r}
    constraintSetup [{transformType=Rotation, targetBone {boneName=upperarm_r}}, {transformType=Rotation, targetBone {boneName=clavicle_r}}]
    constraintWeights [1,1]
   }
   n11 = "Transform (Modify) Bone - Bone: upperarm_twist_02_l"(ComponentPose=@n10, Alpha=1, Translation="0,0,0", Rotation="0,0,0", Scale="1,1,1.4") {
    scaleMode=BMM_Additive
    scaleSpace=BCS_BoneSpace
    alphaBoolBlend {blendOption=Linear}
    alphaScaleBiasClamp {bInterpResult=true}
    boneToModify {boneName=upperarm_twist_02_l}
    scale {z=1.4}
   }
   n12 = "Transform (Modify) Bone - Bone: dyn_belt_03_r"(ComponentPose=@n11, Alpha=1, Translation="0,0,-1", Rotation="5,0,-10", Scale="1,1,1") {
    rotationMode=BMM_Additive
    rotationSpace=BCS_BoneSpace
    translationMode=BMM_Additive
    translationSpace=BCS_BoneSpace
    alphaBoolBlend {blendOption=Linear}
    boneToModify {boneName=dyn_belt_03_r}
    rotation {pitch=5, roll=-10}
    translation {z=-1}
   }
   Save cached pose 'CRCache_PostFixes'(Pose=Component To Local(ComponentPose=@n12)) {cacheName=CRCache_PostFixes}
   n13 = "Transform (Modify) Bone - Bone: base_beltSymbol_m"(
    ComponentPose=Local To Component(LocalPose=Use cached pose 'CRCache_Entry'())
    Alpha=1
    Translation="0,0,0.3"
    Rotation="2,0,0"
    Scale="1,1,1"
   ) {
    rotationMode=BMM_Additive
    rotationSpace=BCS_BoneSpace
    translationMode=BMM_Additive
    translationSpace=BCS_BoneSpace
    alphaBoolBlend {blendOption=Linear}
    boneToModify {boneName=base_beltSymbol_m}
    rotation {pitch=2}
    translation {z=0.3}
   }
   n14 = "Transform (Modify) Bone - Bone: dyn_belt_01_r"(ComponentPose=@n13, Alpha=1, Translation="0,0,0.6", Rotation="0,0,0", Scale="1,1,1") {
    rotationMode=BMM_Additive
    rotationSpace=BCS_BoneSpace
    translationMode=BMM_Additive
    translationSpace=BCS_BoneSpace
    alphaBoolBlend {blendOption=Linear}
    boneToModify {boneName=dyn_belt_01_r}
    translation {z=0.6}
   }
   n15 = "Transform (Modify) Bone - Bone: base_beltSymbol_m"(
    ComponentPose=Local To Component(LocalPose=Use cached pose 'CRCache_Entry'())
    Alpha=1
    Translation="0,0,0.7"
    Rotation="5,0,0"
    Scale="1,1,1"
   ) {
    rotationMode=BMM_Additive
    rotationSpace=BCS_BoneSpace
    translationMode=BMM_Additive
    translationSpace=BCS_BoneSpace
    alphaBoolBlend {blendOption=Linear}
    boneToModify {boneName=base_beltSymbol_m}
    rotation {pitch=5}
    translation {z=0.7}
   }
   n16 = "Transform (Modify) Bone - Bone: dyn_belt_01_r"(ComponentPose=@n15, Alpha=1, Translation="0,0,1", Rotation="10,0,0", Scale="1,1,1") {
    rotationMode=BMM_Additive
    rotationSpace=BCS_BoneSpace
    translationMode=BMM_Additive
    translationSpace=BCS_BoneSpace
    alphaBoolBlend {blendOption=Linear}
    boneToModify {boneName=dyn_belt_01_r}
    rotation {pitch=10}
    translation {z=1}
   }
   n17 = "Transform (Modify) Bone - Bone: dyn_belt_01_l"(ComponentPose=@n16, Alpha=1, Translation="0,0,0", Rotation="10,0,0", Scale="1,1,1") {
    rotationMode=BMM_Additive
    rotationSpace=BCS_BoneSpace
    translationMode=BMM_Additive
    translationSpace=BCS_BoneSpace
    alphaBoolBlend {blendOption=Linear}
    boneToModify {boneName=dyn_belt_01_l}
    rotation {pitch=10}
   }
   n18 = "Transform (Modify) Bone - Bone: base_beltSymbol_m"(ComponentPose=@n17, Alpha=1, Translation="0,0,1", Rotation="8,0,0", Scale="1,1,1") {
    rotationMode=BMM_Additive
    rotationSpace=BCS_BoneSpace
    translationMode=BMM_Additive
    translationSpace=BCS_BoneSpace
    alphaBoolBlend {blendOption=Linear}
    alphaScaleBiasClamp {bInterpResult=true}
    boneToModify {boneName=base_beltSymbol_m}
    rotation {pitch=8}
    translation {z=1}
   }
   n19 = "Transform (Modify) Bone - Bone: dyn_belt_01_r"(ComponentPose=@n18, Alpha=1, Translation="0,0,1", Rotation="10,-5,0", Scale="1,1,1") {
    rotationMode=BMM_Additive
    rotationSpace=BCS_BoneSpace
    translationMode=BMM_Additive
    translationSpace=BCS_BoneSpace
    alphaBoolBlend {blendOption=Linear}
    alphaScaleBiasClamp {bInterpResult=true}
    boneToModify {boneName=dyn_belt_01_r}
    rotation {pitch=10, yaw=-5}
    translation {z=1}
   }
   n20 = "Transform (Modify) Bone - Bone: dyn_belt_01_l"(ComponentPose=@n19, Alpha=1, Translation="0,0,0", Rotation="10,15,0", Scale="1,1,1") {
    rotationMode=BMM_Additive
    rotationSpace=BCS_BoneSpace
    translationMode=BMM_Additive
    translationSpace=BCS_BoneSpace
    alphaBoolBlend {blendOption=Linear}
    alphaScaleBiasClamp {bInterpResult=true}
    boneToModify {boneName=dyn_belt_01_l}
    rotation {pitch=10, yaw=15}
   }
   n21 = Look At - Bone: dyn_belt_01_l(ComponentPose=@n20, Alpha=1) {
    alphaBoolBlend {blendOption=Linear}
    alphaScaleBiasClamp {bInterpResult=true, interpSpeedDecreasing=5, interpSpeedIncreasing=5}
    boneToModify {boneName=dyn_belt_01_l}
    lookAtLocation=<3.9851 14.3094 -0.0027>
    lookAtTarget {boneReference {boneName=calf_l}}
    lookAt_Axis {axis=<1 0>}
   }
   n22 = Look At - Bone: dyn_belt_01_l(ComponentPose=@n21, Alpha=1) {
    alphaBoolBlend {blendOption=Linear}
    alphaScaleBiasClamp {bInterpResult=true, interpSpeedDecreasing=5, interpSpeedIncreasing=5}
    boneToModify {boneName=dyn_belt_01_l}
    lookAtLocation=<3.9851 14.3094 -0.0027>
    lookAtTarget {boneReference {boneName=calf_l}}
    lookAt_Axis {axis=<1 0>}
   }
   n23 = Look At - Bone: dyn_belt_01_l(ComponentPose=@n22, Alpha=1) {
    alphaBoolBlend {blendOption=Linear}
    alphaScaleBiasClamp {bInterpResult=true, interpSpeedDecreasing=5, interpSpeedIncreasing=5}
    boneToModify {boneName=dyn_belt_01_l}
    lookAtLocation=<3.9851 14.3094 -0.0027>
    lookAtTarget {boneReference {boneName=calf_l}}
    lookAt_Axis {axis=<1 0>}
   }
   n24 = Look At - Bone: dyn_belt_01_r(ComponentPose=@n23, Alpha=1) {
    alphaBoolBlend {blendOption=Linear}
    alphaScaleBiasClamp {bInterpResult=true, interpSpeedDecreasing=5, interpSpeedIncreasing=5}
    boneToModify {boneName=dyn_belt_01_r}
    lookAtLocation=<-8.6482 -14.229>
    lookAtTarget {boneReference {boneName=calf_r}}
    lookAt_Axis {axis=<1 0>}
   }
   n25 = Look At - Bone: dyn_belt_01_r(ComponentPose=@n24, Alpha=1) {
    alphaBoolBlend {blendOption=Linear}
    alphaScaleBiasClamp {bInterpResult=true, interpSpeedDecreasing=5, interpSpeedIncreasing=5}
    boneToModify {boneName=dyn_belt_01_r}
    lookAtLocation=<-8.6308 9.3963 1.0609>
    lookAtTarget {boneReference {boneName=calf_r}}
    lookAt_Axis {axis=<1 0>}
   }
   Save cached pose 'CRCache_prePoss'(Pose=Blend Poses by bool(
     BlendPose_0=Component To Local(ComponentPose=@n14)
     BlendPose_1=Component To Local(ComponentPose=@n25)
     bActiveValue=Get isfrontEndBlend() -> (isfrontEndBlend:bool)
    ) {blendType=Linear}) {cacheName=CRCache_prePoss}
   n26 = RigidBody(ComponentPose=Local To Component(LocalPose=Use cached pose 'CRCache_prePoss'())) {
    bOverrideWorldGravity=true
    overridePhysicsAsset=&/M_MED_Snow_Ninja_Belt_FrontEnd_Physics
    simulationSpace=BaseBoneSpace
    alphaBoolBlend {blendOption=Linear}
    baseBoneRef {boneName=pelvis}
    componentAppliedLinearAccClamp=<3000 3000 3000>
    componentLinearAccScale=<0.4 0.4 0.4>
    componentLinearVelScale=<0.4 0.4 0.4>
    overrideWorldGravity {z=-980}
    simSpaceSettings {
     maxAngularAcceleration=1000
     maxAngularVelocity=3
     maxLinearAcceleration=700
     maxLinearVelocity=500
     velocityScaleZ=0
     worldAlpha=0.2
     externalLinearDragV=<0.2 0.2>
    }
   }
   n27 = RigidBody(ComponentPose=Local To Component(LocalPose=Use cached pose 'CRCache_prePoss'())) {
    bOverrideWorldGravity=true
    overridePhysicsAsset=&/M_MED_Snow_Ninja_Belt_Physics
    simulationSpace=BaseBoneSpace
    alphaBoolBlend {blendOption=Linear}
    baseBoneRef {boneName=pelvis}
    componentAppliedLinearAccClamp=<3000 3000 3000>
    componentLinearAccScale=<0.4 0.4 0.4>
    componentLinearVelScale=<0.4 0.4 0.4>
    overrideWorldGravity {z=-980}
    simSpaceSettings {
     maxAngularAcceleration=1000
     maxAngularVelocity=3
     maxLinearAcceleration=700
     maxLinearVelocity=500
     velocityScaleZ=0
     worldAlpha=0.2
     externalLinearDragV=<0.2 0.2>
    }
   }
   Save cached pose 'CRCache_BeltDyn'(Pose=Blend Poses by bool(BlendPose_0=Component To Local(ComponentPose=@n26), BlendPose_1=Component To Local(ComponentPose=@n27), bActiveValue=Get isfrontEndBlend()) {blendType=Linear}) {cacheName=CRCache_BeltDyn}
   n28 = Get isParachuting() -> (isParachuting:double)
   n29 = "Transform (Modify) Bone - Bone: dyn_collar_02_l"(
    ComponentPose=Local To Component(LocalPose=Use cached pose 'CRCache_PostFixes'())
    Alpha=@n28
    Translation="8.5,2,0"
    Rotation="0,0,0"
    Scale="1,1,0.7"
   ) {
    scaleMode=BMM_Additive
    scaleSpace=BCS_BoneSpace
    translationMode=BMM_Additive
    translationSpace=BCS_BoneSpace
    alphaBoolBlend {blendOption=Linear}
    alphaScaleBiasClamp {bInterpResult=true}
    boneToModify {boneName=dyn_collar_02_l}
    scale {z=0.7}
    translation=<8.5 2>
   }
   n30 = "Transform (Modify) Bone - Bone: dyn_collar_02_r"(ComponentPose=@n29, Alpha=@n28, Translation="-8,0,0", Rotation="0,0,0", Scale="1,1,0.8") {
    rotationMode=BMM_Additive
    rotationSpace=BCS_BoneSpace
    scaleMode=BMM_Additive
    scaleSpace=BCS_BoneSpace
    translationMode=BMM_Additive
    translationSpace=BCS_BoneSpace
    alphaBoolBlend {blendOption=Linear}
    alphaScaleBiasClamp {bInterpResult=true}
    boneToModify {boneName=dyn_collar_02_r}
    scale {z=0.8}
    translation {x=-8}
   }
   n31 = "Transform (Modify) Bone - Bone: base_shoulderPad_b0_r"(ComponentPose=@n30, Alpha=@n28, Translation="0,0,5", Rotation="0,0,0", Scale="1,1,1") {
    rotationMode=BMM_Additive
    rotationSpace=BCS_BoneSpace
    scaleMode=BMM_Additive
    scaleSpace=BCS_BoneSpace
    translationMode=BMM_Additive
    translationSpace=BCS_BoneSpace
    alphaBoolBlend {blendOption=Linear}
    alphaScaleBiasClamp {bInterpResult=true}
    boneToModify {boneName=base_shoulderPad_b0_r}
    translation {z=5}
   }
   n32 = "Transform (Modify) Bone - Bone: base_shoulderPad_b0_l"(ComponentPose=@n31, Alpha=@n28, Translation="0,0,5", Rotation="0,0,0", Scale="1,1,1") {
    rotationMode=BMM_Additive
    rotationSpace=BCS_BoneSpace
    scaleMode=BMM_Additive
    scaleSpace=BCS_BoneSpace
    translationMode=BMM_Additive
    translationSpace=BCS_BoneSpace
    alphaBoolBlend {blendOption=Linear}
    alphaScaleBiasClamp {bInterpResult=true}
    boneToModify {boneName=base_shoulderPad_b0_l}
    translation {z=5}
   }
   n33 = "Transform (Modify) Bone - Bone: dyn_collar_02_r"(ComponentPose=@n32, Alpha=1, Translation="-3.5,0,0", Rotation="0,-20,0", Scale="1,1,1") {
    rotationMode=BMM_Additive
    rotationSpace=BCS_BoneSpace
    scaleMode=BMM_Additive
    scaleSpace=BCS_BoneSpace
    translationMode=BMM_Additive
    translationSpace=BCS_BoneSpace
    alphaBoolBlend {blendOption=Linear}
    alphaScaleBiasClamp {bInterpResult=true}
    boneToModify {boneName=dyn_collar_02_r}
    rotation {yaw=-20}
    translation {x=-3.5}
   }
   n34 = "Transform (Modify) Bone - Bone: dyn_collar_02_l"(ComponentPose=@n33, Alpha=1, Translation="5.5,0,0", Rotation="0,-20,0", Scale="1,1,1") {
    rotationMode=BMM_Additive
    rotationSpace=BCS_BoneSpace
    scaleMode=BMM_Additive
    scaleSpace=BCS_BoneSpace
    translationMode=BMM_Additive
    translationSpace=BCS_BoneSpace
    alphaBoolBlend {blendOption=Linear}
    alphaScaleBiasClamp {bInterpResult=true}
    boneToModify {boneName=dyn_collar_02_l}
    rotation {yaw=-20}
    translation {x=5.5}
   }
   Save cached pose 'CRCache_PostFixesHoodie'(Pose=Component To Local(ComponentPose=@n34)) {cacheName=CRCache_PostFixesHoodie}
  ]}
 EventGraph {nodes [
   Event BlueprintUpdateAnimation()
   > n1 = Sequence!(execute) -> (then_0, then_1, then_2:Exec, then_3:Exec, then_4:Exec, then_5:Exec, then_6:Exec)
   > n2 = Cast To FortPlayerAnimInstance(Object=GetAnimInstance(self=GetOwnerSkeletalMeshComponent()))
   > Set GoingCommandoGun(GoingCommandoGun:bool=false) -> (Output_Get:bool)
   > Set dualWeildAlpha(dualWeildAlpha:bool=false) -> (Output_Get:bool)
   > Set LargeGun(LargeGun:bool=false) -> (Output_Get:bool)
   > Set rocketLauncher(rocketLauncher:bool=false) -> (Output_Get:bool)
   > Set pistolAlpha(pistolAlpha:bool=false) -> (Output_Get:bool)
   > Set AlphaShotgun(AlphaShotgun:float=0) -> (Output_Get:float)
   > Set ShotGun(ShotGun:bool=false) -> (Output_Get:bool)
   > Set AR_BUllPup(AR_BUllPup:bool=false) -> (Output_Get:bool)
   > n3 = Switch on EFortWeaponCoreAnimation!(execute, Selection:EFortWeaponCoreAnimation Enum=Get WeaponCoreAnim(self=@n2)) -> (
    Melee:Exec
    Pistol:Exec
    Shotgun:Exec
    PaperBlueprint:Exec
    Rifle:Exec
    MeleeOneHand:Exec
    MachinePistol:Exec
    RocketLauncher:Exec
    GrenadeLauncher:Exec
    GoingCommando:Exec
    AssaultRifle:Exec
    TacticalShotgun:Exec
    SniperRifle:Exec
    TrapPlacement:Exec
    ShoulderLauncher:Exec
    AbilityDecoTool:Exec
    Crossbow:Exec
    C4:Exec
    RemoteControl:Exec
    DualWield:Exec
    AR_BullPup:Exec
    AR_ForwardGrip:Exec
    MedPackPaddles:Exec
    SMG_P90:Exec
    AR_DrumGun:Exec
    Consumable_Small:Exec
    Consumable_Large:Exec
    Balloon:Exec
    MountedTurret:Exec
    CreativeTool:Exec
    ExplosiveBow:Exec
    AshtonIndigo:Exec
    AshtonChicago:Exec
    MeleeDualWield:Exec
    Unarmed:Exec
   )
   Set AR_BUllPup(execute=@n3.AR_BullPup, AR_BUllPup=true)
   Set LargeGun(execute=@n3.AssaultRifle, LargeGun=true)
   Set dualWeildAlpha(execute=@n3.DualWield, dualWeildAlpha=true)
   Set GoingCommandoGun(execute=@n3.GoingCommando, GoingCommandoGun=true)
   Set AlphaShotgun(execute=@n3.Pistol, AlphaShotgun=1)
   > Set pistolAlpha(pistolAlpha=true)
   Set rocketLauncher(execute=@n3.RocketLauncher, rocketLauncher=true)
   Set ShotGun(execute=[@n3.Shotgun, @n3.TacticalShotgun], ShotGun=true)
   n4 = BreakTransform(InTransform=GetDeltaTransformFromRefPose(self=GetOwnerSkeletalMeshComponent(), BoneName=upperarm_r, BaseName=clavicle_r))
   n5 = Branch(execute=@n1.then_3, Condition="float >= float"(A=BreakRotator(InRot=@n4.Rotation).Yaw, B=95))
   > Set UpperArmRot_l(UpperArmRot_l:bool=1) -> (Output_Get:bool)
   Set UpperArmRot_l(execute=@n5.else, UpperArmRot_l=false)
   n6 = Sequence(execute=@n1.then_1)
   n7 = Branch(execute=@n6.then_1, Condition=Get LargeGun() -> (LargeGun:bool))
   > Set LargeGunOut(LargeGunOut:double=1) -> (Output_Get:double)
   Set LargeGunOut(execute=@n7.else, LargeGunOut=0)
   n8 = Branch(execute=@n6.then_2, Condition=Get GoingCommandoGun() -> (GoingCommandoGun:bool))
   > Set BigGunOut(BigGunOut:double=1) -> (Output_Get:double)
   Set BigGunOut(execute=@n8.else, BigGunOut=0)
   n9 = Branch(execute=@n6.then_3, Condition=Get rocketLauncher() -> (rocketLauncher:bool))
   > Set RocketGunOut(RocketGunOut:double=1) -> (Output_Get:double)
   Set RocketGunOut(execute=@n9.else, RocketGunOut=0)
   n10 = Branch(execute=@n6.then_4, Condition=Get ShotGun() -> (ShotGun:bool))
   > Set ShotgunAlpha(ShotgunAlpha:double=1) -> (Output_Get:double)
   Set ShotgunAlpha(execute=@n10.else, ShotgunAlpha=0)
   n11 = Branch(execute=@n6.then_5, Condition=Get AR_BUllPup() -> (AR_BUllPup:bool))
   > Set AR_BullAlpha(AR_BullAlpha:double=1) -> (Output_Get:double)
   Set AR_BullAlpha(execute=@n11.else, AR_BullAlpha=0)
   n12 = Branch(execute=@n6.then_0, Condition=Get dualWeildAlpha() -> (dualWeildAlpha:bool))
   > Set dualGunFix(dualGunFix:double=0) -> (Output_Get:double)
   > Set dualGunOut(dualGunOut:double=1) -> (Output_Get:double)
   Set dualGunFix(execute=@n12.else, dualGunFix=1)
   > Set dualGunOut(dualGunOut=0)
   n13 = Sequence!(execute=@n1.then_2) -> (then_0, then_1)
   > n14 = Branch(Condition=OR Boolean!(
     A=Get GoingCommandoGun()
     B=Get rocketLauncher()
     C:bool=Get LargeGun()
     D:bool=Get ShotGun()
     E:bool=Get pistolAlpha() -> (pistolAlpha:bool)
    ) -> (ReturnValue))
   > n15 = Branch(Condition=Get pistolAlpha())
   > Set CollarFixAlpha(CollarFixAlpha:double=0.5) -> (Output_Get:double)
   Set CollarFixAlpha(execute=@n15.else, CollarFixAlpha=0)
   Set CollarFixAlpha(execute=@n14.else, CollarFixAlpha=1)
   n16 = Branch(execute=@n13.then_1, Condition=Get bIsCrouching())
   > Set crouchFixAlpha(crouchFixAlpha:double=0) -> (Output_Get:double)
   Set crouchFixAlpha(execute=@n16.else, crouchFixAlpha=1)
   n17 = Get pistolAlpha()
   n18 = Get UpperArmRot_l() -> (UpperArmRot_l:bool)
   Set aimUp(execute=@n1.then_4, aimUp:double="To Float (Boolean)"(InBool=AND Boolean(A=@n17, B=@n18))) -> (Output_Get:double)
   > Set crouchAlpha(crouchAlpha:double="To Float (Boolean)"(InBool=Get bIsCrouching())) -> (Output_Get:double)
   > Set isParachuting(isParachuting:double="To Float (Boolean)"(InBool=Get bIsParachuteOpened())) -> (Output_Get:double)
   > Set isFrontEndAlpha(isFrontEndAlpha:double="To Float (Boolean)"(InBool=Get bIsInFrontEnd())) -> (Output_Get:double)
   n19 = Get bIsPlayingEmote()
   > n20 = Branch(Condition=AND Boolean(A=Get bIsInFrontEnd(), B=NOT Boolean(A=@n19)))
   > Set isfrontEndBlend(isfrontEndBlend:bool=true) -> (Output_Get:bool)
   Set isfrontEndBlend(execute=@n20.else, isfrontEndBlend=false)
   n21 = PlayerMovement(execute=@n1.then_5, AnimInstance=Self-Reference(), XDotProductCoefficient=100, YDotProductCoefficient=100, ZDotProductCoefficient=100)
   n22 = NOT Boolean(A=Get bIsCrouching())
   n23 = BreakVector(InVec=MakeVector(X=@n21, Y=@n21.YMovement))
   n24 = "float > float"(A=@n23.Y, B=40)
   > Set Forward(Forward:double="To Float (Boolean)"(InBool=AND Boolean(A=@n24, B=@n22))) -> (Output_Get:double)
   n25 = "float < float"(A=@n23, B=-50)
   n26 = Get bIsCrouching()
   > Set Left(Left:double="To Float (Boolean)"(InBool=AND Boolean(A=@n25, B=@n26))) -> (Output_Get:double)
   n27 = "float > float"(A=@n23, B=50)
   n28 = Get bIsCrouching()
   > Set Right(Right:double="To Float (Boolean)"(InBool=AND Boolean(A=@n27, B=@n28))) -> (Output_Get:double)
  ]}
]
variables [
 aimUp {type=number}
 AlphaShotgun {type=number}
 AR_BullAlpha {type=number}
 AR_BUllPup {type=boolean}
 BigGunOut {type=number}
 CollarFixAlpha {type=number}
 crouchAlpha {type=number}
 crouchFixAlpha {type=number}
 dualGunFix {type=number}
 dualGunOut {type=number}
 dualWeildAlpha {type=boolean}
 Forward {type=number}
 GoingCommandoGun {type=boolean}
 isFrontEndAlpha {type=number}
 isfrontEndBlend {type=boolean}
 isParachuting {type=number}
 LargeGun {type=boolean}
 LargeGunOut {type=number}
 Left {type=number}
 pistolAlpha {type=boolean}
 Right {type=number}
 RocketGunOut {type=number}
 rocketLauncher {type=boolean}
 ShotGun {type=boolean}
 ShotgunAlpha {type=number}
 UpperArmRot_l {type=boolean}
]
